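; Client announcement Blueprint in the homebase onboarding front end that plays an event cinematic with subtitles.
; On entering the cinematic state it wraps the event media source in a FortMediaInfo and fires a QueueCinematic client event with the subtitles player.
; A one-second timer enables skipping, after which a player skip fires SkipCinematic.
; The CinematicFinishedPlaying client event disables skipping and stops the announcement, guarded by a DoOnce.
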(event Announcement|Cinematic|EventOnEnteredCinematicState
  ;[1 1 1] Start of Satellite Cinematic
  (bind _returnvalue (Game|ConstructObjectfromClass "/Script/FortniteUI.FortMediaInfo" self 0 (Variables|Default|GetEventMediaSource) false))
  (CallFunction|ClientEvent|ExecuteClientEvent "(CategoryName=\"Cinematic\",EventName=\"QueueCinematic\")" _returnvalue (Variables|Default|GetEventSubtitlesPlayer))
  (Utilities|Time|SetTimerbyEvent (EventDispatchers|CreateEvent :!event Custom|HandleAllowSkip) 1.0))

(event Custom|HandleAllowSkip
  ;[1 1 1] Skipping Allowed
  (Variables|Default|SetAllowSkipping true))

(event Announcement|Cinematic|EventOnPlayerSkippedCutscene
  ;[1 1 1] Skipped
  (if (Variables|Default|GetAllowSkipping)
    (CallFunction|ClientEvent|ExecuteClientEvent "(CategoryName=\"Cinematic\",EventName=\"SkipCinematic\")")))

(event ClientEvent|HandleClientEvent:CinematicFinishedPlaying
  (params (EventSource :type Object%20Reference) (EventFocus :type Object%20Reference) (ClientEvent :type Fort%20Client%20Event%20Structure%20%28by%20ref%29))
  ;[1 1 1] Finished playing cinematic
  (Utilities|FlowControl|DoOnce false
    (:Completed
      (Variables|Default|SetAllowSkipping false)
      (Announcement|StopClientAnnouncement))))

(event Announcement|EventOnClientAnnouncementStop)

(fn ConstructionScript)
